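# Additive unlit particle sprite material for the Haunt player character's chest glow, layering tinted lightning, fog and smoke.
# Noise-wobbled sphere masks and a swirl texture twist the lightning UVs, which get bump offset and feed emissive, modulated by smoke.
# Deferred shading reads scene world normals to distort UVs and shape a view-dependent mask gated by custom depth and depth fade.
# Forward shading simplifies this, mobile uses a tinted flare texture, and sprites are camera-offset.
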Material /Game/Characters/Player/Male/Medium/Bodies/M_MED_Haunt/FX/M_ChestGlow
graph {
 properties {
  bUseMaterialAttributes=true
  bUsedWithParticleSprites=true
  blendMode=BLEND_Additive
  pixelDepthOffsetMode=PDOM_Legacy
  shadingModel=MSM_Unlit
  usedShadingModels=Unlit
 }
 nodes [
  n1 = ParticleColor()
  n2 = MaterialFunctionCall(Offset Amount=ScalarParameter() {defaultValue=21, parameterName=CameraOffset}) {materialFunction=/Engine/Functions/Engine_MaterialFunctions02/WorldPositionOffset/CameraOffset}
  n3 = TextureSample(UVs=MaterialFunctionCall(Texture Scale=ScalarParameter() {defaultValue=0.5, parameterName=Twist_Size}) {materialFunction=&ScaleUVsByCenter}.UVs) {samplerType=SAMPLERTYPE_LinearColor, texture=/Game/Effects/Fort_Effects/Textures/Taker/T_TakerPortal_CoreFlow}
  n4 = Constant() {r=0.5}
  n5 = Add(A=TextureCoordinate() {uTiling=0.1, vTiling=0.1}, B=Time())
  n6 = Subtract(
   A=ComponentMask(None=TextureSample(UVs=@n5) {texture=/Game/Global/Textures/Default/Noise/3xPerlin_M}.RGB) {g=true, r=true}
   B=ScalarParameter() {defaultValue=0.3, parameterName=Turbulence_Offset}
  ) {constB=0.5}
  n7 = ScalarParameter() {defaultValue=0.1, parameterName=Turbulence}
  n8 = SphereMask(A=Add(A=TextureCoordinate(), B=Multiply(A=@n6, B=@n7) {constB=0.2}), B=@n4, Radius=@n4, Hardness=Constant())
  n9 = Multiply(A=ComponentMask(None=Multiply(A=@n3.RGB)) {g=true, r=true}, B=Multiply(A=@n8, B=ScalarParameter() {defaultValue=0.1, parameterName=Twist}))
  n10 = VectorParameter() {parameterName=LightningTillingAndSpeed, defaultValue=<rgba 1 1.5 0.6 0.2>}
  n11 = ComponentMask(None=@n10.RGB) {g=true, r=true}
  n12 = Add(A=TextureCoordinate(), B=@n9)
  n13 = Time()
  n14 = AppendVector(A=@n10.B, B=@n10.B)
  n15 = Multiply(A=Add(A=Multiply(A=@n11, B=@n12), B=Multiply(A=@n13, B=@n14)))
  n16 = SceneTexture() {sceneTextureId=PPI_WorldNormal}
  n17 = ComponentMask(None=Transform(None=ComponentMask(None=@n16.Color) {b=true, g=true, r=true}) {transformSourceType=TRANSFORMSOURCE_World, transformType=TRANSFORM_Camera}) {g=true, r=true}
  n18 = ScalarParameter() {defaultValue=0.1, parameterName=Lightning_NormalDistort}
  n19 = Add(A=@n9, B=Add(A=@n15, B=Multiply(A=@n17, B=@n18)))
  n20 = Add(A=@n15, B=@n9)
  n21 = TextureSample(UVs=BumpOffset(
    Coordinate=ShadingPathSwitch(Default=@n19, Forward=@n20)
    Height=@n8
    HeightRatioInput=ScalarParameter() {defaultValue=-1, parameterName=Lighting_Bump}
   )) {samplerType=SAMPLERTYPE_Masks, texture=/Game/Weapons/WeaponSkins/Wraps/Textures/PlasmaCharge/PlasmaCharge_M}
  n22 = DepthFade(FadeDistance=ScalarParameter() {defaultValue=150, parameterName=DF_FD})
  n23 = Sine(None=Multiply(A=@n22) {constB=0.5})
  n24 = SceneTexture() {sceneTextureId=PPI_CustomDepth}
  n25 = Multiply(A=@n23, B=Saturate(None=Floor(None=@n24.Color)))
  n26 = ComponentMask(None=SceneTexture() {sceneTextureId=PPI_WorldNormal}.Color) {b=true, g=true, r=true}
  n27 = CameraVectorWS()
  n28 = Multiply(A=@n8, B=@n8)
  n29 = Multiply(
   A=LinearInterpolate(A=DotProduct(A=@n26, B=@n27), B=ScalarParameter() {defaultValue=1.2, parameterName=AlphaBais}, Alpha=@n28) {constB=2}
   B=@n28
  )
  n30 = Constant() {r=0.5}
  n31 = Multiply(A=SphereMask(A=TextureCoordinate(), B=@n30, Radius=@n30) {hardnessPercent=0}, B=@n23) {constB=0.2}
  n32 = ComponentMask(None=ShadingPathSwitch(Default=Multiply(A=@n25, B=@n29), Forward=Multiply(A=@n31) {constB=0.22})) {r=true}
  n33 = Power(
   Base=Multiply(A=Multiply(A=@n21.G, B=@n32), B=VectorParameter() {parameterName=LightningTint, defaultValue=<rgba 25.625 34.8319 35 0>}.RGB)
   Exp=ScalarParameter() {defaultValue=5, parameterName=LightningTintExp}
  )
  n34 = Multiply(A=@n32, B=VectorParameter() {parameterName=FogTint, defaultValue=<rgba 1.2965 2.714 3.7 0>}.RGB)
  n35 = ScalarParameter() {defaultValue=2.2, parameterName=FogTintExp}
  n36 = ScalarParameter() {defaultValue=-0.1, parameterName=CloudLerp1}
  n37 = ScalarParameter() {defaultValue=40, parameterName=CloudLerp2}
  n38 = TextureCoordinate()
  n39 = ScalarParameter() {defaultValue=-0.07, parameterName=Cloud_NormalDistort}
  n40 = ShadingPathSwitch(Default=Add(A=@n38, B=Multiply(A=@n39, B=@n17)), Forward=@n38)
  n41 = VectorParameter() {parameterName=CloudMask, defaultValue=<rgba 0.5 0.5 0.2 0.2>}
  n42 = ComponentMask(None=@n41.RGB) {g=true, r=true}
  n43 = TextureSample(UVs=Panner(Coordinate=Multiply(A=@n40, B=@n42), Speed=AppendVector(A=@n41.B, B=@n41.B))) {texture=/Game/Global/Textures/Default/Tileable_Patterns/T_Thick_Smoke_05}
  n44 = Multiply(
   A=Add(A=LinearInterpolate(A=@n36, B=@n37, Alpha=@n43.R) {constA=2, constB=-0.5}, B=Multiply(A=@n8) {constB=0.2})
   B=Add(A=Multiply(A=@n33, B=@n32), B=Power(Base=@n34, Exp=@n35))
  )
  n45 = TextureSampleParameter2D() {parameterName=FlareMask, texture=/Game/Effects/Fort_Effects/Textures/Flares/T_Flare_Spikeball}
  n46 = VectorParameter() {parameterName=Timt, defaultValue=<rgba 0.5625 1 0.9057 1>}
  n47 = ParticleColor()
  n48 = Multiply(A=Multiply(A=Multiply(A=@n45.RGB, B=@n46.RGB)) {constB=0.3}, B=@n47.RGB)
  n49 = MaterialFunctionCall(Offset Amount=ScalarParameter() {defaultValue=30, parameterName=CameraOffset_Low}) {materialFunction=&CameraOffset}
  MaterialFunctionCall!(In=ShadingPathSwitch(
    Default=SetMaterialAttributes!(MaterialAttributes, Emissive Color=@n44, Opacity=@n1.A, World Position Offset=@n2.Clamped Camera Offset) {attributeSetTypes [#1, #2, #3]}
    Mobile=SetMaterialAttributes!(MaterialAttributes, Emissive Color=@n48, World Position Offset=@n49.Clamped Camera Offset, Opacity=@n47.A) {attributeSetTypes [#1, #3, #2]}
   ) {desc=Migrated from FeatureLevelSwitch}) {materialFunction=/Game/Effects/Niagara/MasterMaterials/Functions/MF_FX_EmissiveAdjustments_MA}
  OneMinus(None=@n22)
 ]
}
